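# Physics asset for the hair on the F_MED_PirouetteWeld female player cosmetic, defining simulation bodies for the bun, side locks and front strands.
# Each hair bone carries one capsule, while the head carries five overlapping capsules approximating its shape.
# Swing and twist limited constraints form branching chains rooted at the head.
# Limits generally loosen toward the tips, with front strand roots held nearly rigid.
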PhysicsAsset /BRCosmetics/Characters/Player/Female/Medium/Bodies/F_MED_PirouetteWeld/Meshes/Parts/F_MED_PirouetteWeld_FaceAcc_Hair_Physics
bodies {
 dyn_hair_bun [{shapeType=Capsule, length=0.75, radius=12, shapeName=hairBun_rt_00_capsule, center=<12 0 0>, rotation=<rot 90 0 0>}]
 dyn_hair_side_l_01 [{shapeType=Capsule, length=2.1457, radius=1.8543, shapeName=dyn_hair_side_l_01_capsule, center=<1.5 0 3>, rotation=<rot 90 0 0>}]
 dyn_hair_side_l_02 [{shapeType=Capsule, length=1.6767, radius=2.3233, shapeName=dyn_hair_side_l_02_capsule, center=<1.5 0 3>, rotation=<rot 90 0 0>}]
 dyn_hair_side_l_03 [{shapeType=Capsule, length=0.5748, radius=3.4252, shapeName=dyn_hair_side_l_03_capsule, center=<1.5 0 3>, rotation=<rot 90 0 0>}]
 dyn_hair_side_l_04 [{shapeType=Capsule, length=0.5748, radius=3.4252, shapeName=dyn_hair_side_l_04_capsule, center=<1.5 0 3>, rotation=<rot 90 0 0>}]
 dyn_hair_side_l_05 [{shapeType=Capsule, length=0.5748, radius=3.4252, shapeName=dyn_hair_side_l_05_capsule, center=<1.5 0 3>, rotation=<rot 90 0 0>}]
 dyn_hair_side_l_06 [{shapeType=Capsule, length=0, radius=4.2563, shapeName=dyn_hair_side_l_06_capsule, center=<1.5 0 3>, rotation=<rot 90 0 0>}]
 dyn_hair_side_l_07 [{shapeType=Capsule, length=0.774, radius=3.226, shapeName=dyn_hair_side_l_07_capsule, center=<1.5 0 3>, rotation=<rot 90 0 0>}]
 dyn_hair_side_l_08 [{shapeType=Capsule, length=0.774, radius=3.226, shapeName=dyn_hair_side_l_08_capsule, center=<1.5 0 3>, rotation=<rot 90 0 0>}]
 dyn_hair_side_r_01 [{shapeType=Capsule, length=1.8525, radius=2.1475, shapeName=dyn_hair_side_r_01_capsule, center=<1.5 0 3>, rotation=<rot 90 0 0>}]
 dyn_hair_side_r_02 [{shapeType=Capsule, length=1.8525, radius=2.1475, shapeName=dyn_hair_side_r_02_capsule, center=<1.5 0 3>, rotation=<rot 90 0 0>}]
 dyn_hair_side_r_03 [{shapeType=Capsule, length=1.8525, radius=2.1475, shapeName=dyn_hair_side_r_03_capsule, center=<1.5 0 3>, rotation=<rot 90 0 0>}]
 dyn_hair_side_r_04 [{shapeType=Capsule, length=1.8525, radius=2.1475, shapeName=dyn_hair_side_r_04_capsule, center=<1.5 0 3>, rotation=<rot 90 0 0>}]
 dyn_hair_side_r_05 [{shapeType=Capsule, length=1.8525, radius=2.1475, shapeName=dyn_hair_side_r_05_capsule, center=<1.5 0 3>, rotation=<rot 90 0 0>}]
 dyn_hair_side_r_06 [{shapeType=Capsule, length=1.5434, radius=2.4566, shapeName=dyn_hair_side_r_06_capsule, center=<1.5 0 3>, rotation=<rot 90 0 0>}]
 dyn_hair_side_r_07 [{shapeType=Capsule, length=1.5434, radius=2.4566, shapeName=dyn_hair_side_r_07_capsule, center=<1.5 0 3>, rotation=<rot 90 0 0>}]
 dyn_strand_fr_l_01 [{shapeType=Capsule, length=1.0921, radius=2.9079, shapeName=dyn_strand_fr_l_01_capsule, center=<1.5 0 3>, rotation=<rot 90 0 0>}]
 dyn_strand_fr_l_02 [{shapeType=Capsule, length=0.7225, radius=3.2775, shapeName=dyn_strand_fr_l_02_capsule, center=<1.5 0 3>, rotation=<rot 90 0 0>}]
 dyn_strand_fr_l_03 [{shapeType=Capsule, length=0.7225, radius=3.2775, shapeName=dyn_strand_fr_l_03_capsule, center=<1.5 0 3>, rotation=<rot 90 0 0>}]
 dyn_strand_fr_l_04 [{shapeType=Capsule, length=0.7225, radius=3.2775, shapeName=dyn_strand_fr_l_04_capsule, center=<1.5 0 3>, rotation=<rot 90 0 0>}]
 dyn_strand_fr_l_05 [{shapeType=Capsule, length=0.7225, radius=3.2775, shapeName=dyn_strand_fr_l_05_capsule, center=<1.5 0 3>, rotation=<rot 90 0 0>}]
 dyn_strand_fr_r_01 [{shapeType=Capsule, length=1.2389, radius=2.7611, shapeName=dyn_strand_fr_r_01_capsule, center=<1.5 0 3>, rotation=<rot 90 0 0>}]
 dyn_strand_fr_r_02 [{shapeType=Capsule, length=1.0101, radius=2.9899, shapeName=dyn_strand_fr_r_02_capsule, center=<1.5 0 3>, rotation=<rot 90 0 0>}]
 dyn_strand_fr_r_03 [{shapeType=Capsule, length=1.0101, radius=2.9899, shapeName=dyn_strand_fr_r_03_capsule, center=<1.5 0 3>, rotation=<rot 90 0 0>}]
 dyn_strand_fr_r_a_01 [{shapeType=Capsule, length=1.1019, radius=2.8981, shapeName=dyn_strand_fr_r_a_01_capsule, center=<1.5 0 3>, rotation=<rot 90 0 0>}]
 dyn_strand_fr_r_a_02 [{shapeType=Capsule, length=0.9146, radius=3.0854, shapeName=dyn_strand_fr_r_a_02_capsule, center=<1.5 0 3>, rotation=<rot 90 0 0>}]
 head [
  {shapeType=Capsule, length=12.4251, radius=5.9454, shapeName=head_capsule, center=<0.4349 0.6257 0>, rotation=<rot 90 0 -12.0083>}
  {shapeType=Capsule, length=0, radius=3.6663, shapeName=head_capsule, center=<3.7814 -0.7896 -4.7406>, rotation=<rot 90 0 -12.0083>}
  {shapeType=Capsule, length=4.1681, radius=4.7372, shapeName=head_capsule, center=<0.2092 -2.0824 3.7307>, rotation=<rot 90 0 -12.0083>}
  {shapeType=Capsule, length=1.0928, radius=7.3101, shapeName=head_capsule, center=<5.3553 1.6156 0.0156>, rotation=<rot 90 0 -12.0083>}
  {shapeType=Capsule, length=6.6954, radius=5.6354, shapeName=head_capsule, center=<1.816 3.9683 0.0908>, rotation=<rot 90 180 179.289>}
 ]
}
constraints [
 {bone1Name=dyn_hair_bun, bone2Name=head, swing1LimitDegrees=15, swing2LimitDegrees=15, twistLimitDegrees=1}
 {bone1Name=dyn_hair_side_l_01, bone2Name=head, swing1LimitDegrees=10, swing2LimitDegrees=10, twistLimitDegrees=5}
 {bone1Name=dyn_hair_side_l_02, bone2Name=dyn_hair_side_l_01, swing1LimitDegrees=5, swing2LimitDegrees=5, twistLimitDegrees=5}
 {bone1Name=dyn_hair_side_l_03, bone2Name=dyn_hair_side_l_02, swing1LimitDegrees=10, swing2LimitDegrees=10, twistLimitDegrees=5}
 {bone1Name=dyn_hair_side_l_04, bone2Name=dyn_hair_side_l_03, swing1LimitDegrees=20, swing2LimitDegrees=20, twistLimitDegrees=5}
 {bone1Name=dyn_hair_side_l_05, bone2Name=dyn_hair_side_l_04, swing1LimitDegrees=30, swing2LimitDegrees=20, twistLimitDegrees=5}
 {bone1Name=dyn_hair_side_l_06, bone2Name=dyn_hair_side_l_03, swing1LimitDegrees=20, swing2LimitDegrees=20, twistLimitDegrees=5}
 {bone1Name=dyn_hair_side_l_07, bone2Name=dyn_hair_side_l_02, swing1LimitDegrees=20, swing2LimitDegrees=20, twistLimitDegrees=5}
 {bone1Name=dyn_hair_side_l_08, bone2Name=dyn_hair_side_l_07, swing1LimitDegrees=20, swing2LimitDegrees=20, twistLimitDegrees=5}
 {bone1Name=dyn_hair_side_r_01, bone2Name=head, swing1LimitDegrees=10, swing2LimitDegrees=10, twistLimitDegrees=5}
 {bone1Name=dyn_hair_side_r_02, bone2Name=dyn_hair_side_r_01, swing1LimitDegrees=10, swing2LimitDegrees=10, twistLimitDegrees=5}
 {bone1Name=dyn_hair_side_r_03, bone2Name=dyn_hair_side_r_02, swing1LimitDegrees=25, swing2LimitDegrees=25, twistLimitDegrees=5}
 {bone1Name=dyn_hair_side_r_04, bone2Name=dyn_hair_side_r_03, swing1LimitDegrees=45, swing2LimitDegrees=45, twistLimitDegrees=5}
 {bone1Name=dyn_hair_side_r_05, bone2Name=dyn_hair_side_r_04, swing1LimitDegrees=45, swing2LimitDegrees=45, twistLimitDegrees=5}
 {bone1Name=dyn_hair_side_r_07, bone2Name=dyn_hair_side_r_03, swing1LimitDegrees=20, swing2LimitDegrees=20, twistLimitDegrees=5}
 {bone1Name=dyn_hair_side_r_06, bone2Name=dyn_hair_side_r_02, swing1LimitDegrees=20, swing2LimitDegrees=20, twistLimitDegrees=5}
 {bone1Name=dyn_strand_fr_l_01, bone2Name=head, swing1LimitDegrees=2, swing2LimitDegrees=2, twistLimitDegrees=2}
 {bone1Name=dyn_strand_fr_l_02, bone2Name=dyn_strand_fr_l_01, swing1LimitDegrees=2, swing2LimitDegrees=2, twistLimitDegrees=2}
 {bone1Name=dyn_strand_fr_l_03, bone2Name=dyn_strand_fr_l_02, swing1LimitDegrees=20, swing2LimitDegrees=20, twistLimitDegrees=5}
 {bone1Name=dyn_strand_fr_l_04, bone2Name=dyn_strand_fr_l_01, swing1LimitDegrees=20, swing2LimitDegrees=20, twistLimitDegrees=5}
 {bone1Name=dyn_strand_fr_l_05, bone2Name=dyn_strand_fr_l_04, swing1LimitDegrees=20, swing2LimitDegrees=20, twistLimitDegrees=5}
 {bone1Name=dyn_strand_fr_r_a_01, bone2Name=head, swing1LimitDegrees=2, swing2LimitDegrees=2, twistLimitDegrees=2}
 {bone1Name=dyn_strand_fr_r_a_02, bone2Name=dyn_strand_fr_r_a_01, swing1LimitDegrees=20, swing2LimitDegrees=20, twistLimitDegrees=5}
 {bone1Name=dyn_strand_fr_r_01, bone2Name=head, swing1LimitDegrees=2, swing2LimitDegrees=2, twistLimitDegrees=2}
 {bone1Name=dyn_strand_fr_r_02, bone2Name=dyn_strand_fr_r_01, swing1LimitDegrees=20, swing2LimitDegrees=20, twistLimitDegrees=5}
 {bone1Name=dyn_strand_fr_r_03, bone2Name=dyn_strand_fr_r_02, swing1LimitDegrees=20, swing2LimitDegrees=20, twistLimitDegrees=5}
]
