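# Animation Blueprint for the PirouetteWeld female character's face accessory mesh, adding secondary motion to its hair.
# It copies the attached parent's pose, then bRigidBodyEnabled selects RigidBody simulation relative to spine_05 or chained trail controllers on both side-hair bones.
# Two strand trail controllers are unconnected to the output.
# The event graph updates Gravity Override Cape through a deprecated function, but the RigidBody node uses fixed gravity.
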AnimBlueprint /BRCosmetics/Characters/Player/Female/Medium/Bodies/F_MED_PirouetteWeld/Meshes/Parts/F_MED_PirouetteWeld_FaceAcc_AnimBP
target_skeleton=&/F_MED_PirouetteWeld_FaceAcc_Skeleton
graphs [
 AnimGraph {nodes [
   n1 = Trail controller - Bone: dyn_hair_side_r_05(ComponentPose=Local To Component(LocalPose=Copy Pose From Mesh() {bUseAttachedParent=true}), Alpha=0.6) {
    bLimitRotation=true
    planarLimits [{drivingBone {boneName=spine_05}, planeTransform {location=<0 -10 0>, rotation=<rot 0 20 -90>, scale=<1 1 1>}}]
    rotationLimits [{limitMax=<30 30 30>, limitMin=<-30 -30 -30>}, {limitMax=<30 30 30>, limitMin=<-30 -30 -30>}]
    rotationOffsets [<0 0 0>, <0 0 0>]
    trailBone {boneName=dyn_hair_side_r_05}
    trailRelaxationSpeed {editorCurveData {keys [<at -0.015 6.6299 RCIM_Linear>,<at 0.9488 6.3203 RCIM_Linear>]}}
   }
   n2 = Trail controller - Bone: dyn_hair_side_l_05(ComponentPose=@n1, Alpha=0.6) {
    bLimitRotation=true
    planarLimits [{drivingBone {boneName=spine_05}, planeTransform {location=<0 -10 0>, rotation=<rot 0 20 -90>, scale=<1 1 1>}}]
    rotationLimits [{limitMax=<30 30 30>, limitMin=<-30 -30 -30>}, {limitMax=<30 30 30>, limitMin=<-30 -30 -30>}]
    rotationOffsets [<0 0 0>, <0 0 0>]
    trailBone {boneName=dyn_hair_side_l_05}
    trailRelaxationSpeed {editorCurveData {keys [<at -0.015 6.6299 RCIM_Linear>,<at 0.9488 6.3203 RCIM_Linear>]}}
   }
   n3 = RigidBody!(
    ComponentPose=Local To Component(LocalPose=Copy Pose From Mesh() {bUseAttachedParent=true})
    Alpha
    OverrideWorldGravity:Vector="0,0,-980"
    ExternalForce
   ) -> (Pose) {
    bOverrideWorldGravity=true
    overridePhysicsAsset=&/F_MED_PirouetteWeld_FaceAcc_Hair_Physics
    simulationSpace=BaseBoneSpace
    alphaBoolBlend {blendOption=Linear}
    baseBoneRef {boneName=spine_05}
    componentAppliedLinearAccClamp=<2000 2000 2000>
    componentLinearAccScale=<0.4 0.4 0.5>
    overrideWorldGravity {z=-980}
    simSpaceSettings {
     maxAngularAcceleration=3000
     maxAngularVelocity=3
     maxLinearAcceleration=700
     maxLinearVelocity=500
     velocityScaleZ=0
     worldAlpha=0.5
     externalLinearDragV=<0.2 0.2>
    }
   }
   Output Pose(Result=Blend Poses by bool(BlendPose_0=Component To Local(ComponentPose=@n3), BlendPose_1=Component To Local(ComponentPose=@n2), bActiveValue=Get bRigidBodyEnabled()) {blendType=Linear}) {layerGroup=None}
   n4 = Trail controller - Bone: dyn_strand_02_r(Alpha=0.5) {
    bLimitRotation=true
    planarLimits [{drivingBone {boneName=spine_05}, planeTransform {location=<0 -10 0>, rotation=<rot 0 20 -90>, scale=<1 1 1>}}]
    rotationLimits [{limitMax=<30 30 30>, limitMin=<-30 -30 -30>}, {limitMax=<30 30 30>, limitMin=<-30 -30 -30>}]
    rotationOffsets [<0 0 0>, <0 0 0>]
    trailBone {boneName=dyn_strand_02_r}
    trailRelaxationSpeed {editorCurveData {keys [<at -0.015 6.6299 RCIM_Linear>,<at 0.9488 6.3203 RCIM_Linear>]}}
   }
   Trail controller - Bone: dyn_strand_02_r(ComponentPose=@n4, Alpha=0.5) {
    bLimitRotation=true
    planarLimits [{drivingBone {boneName=spine_05}, planeTransform {location=<0 -10 0>, rotation=<rot 0 20 -90>, scale=<1 1 1>}}]
    rotationLimits [{limitMax=<30 30 30>, limitMin=<-30 -30 -30>}, {limitMax=<30 30 30>, limitMin=<-30 -30 -30>}]
    rotationOffsets [<0 0 0>, <0 0 0>]
    trailBone {boneName=dyn_strand_02_r}
    trailRelaxationSpeed {editorCurveData {keys [<at -0.015 6.6299 RCIM_Linear>,<at 0.9488 6.3203 RCIM_Linear>]}}
   }
  ]}
 EventGraph {nodes [
   Event BlueprintUpdateAnimation()
   > Sequence()
   > n1 = Gravity Override State Function DEPRECATED(CharacterPartAnimInstance=Self-Reference(), GameTimeInSeconds=GetGameTimeInSeconds(), GravityOverrideParams=Get Gravity Override Params())
   > Set Gravity Override Cape(Gravity Override Cape:Vector=@n1) -> (Output_Get:Vector)
  ]}
]
variables [
 Gravity Override Cape {type=Vector}
 Gravity Override Params {type=GravityOverrideParamsStruct}
]
